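; FringePlank prototype weapon Blueprint that, judging by its name, acts as the fire mesh actor for the Curie elemental system.
; Its construction script calls the parent, then evaluates a scalable float row value for self-propagation fuel.
; The result is truncated to an integer and set as the actor's self-propagation fuel, making that fuel amount data-driven.
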(fn ConstructionScript
  (|CallParentFunction)
  (Variables|CurieFuel|SetSelfPropagationFuel (Math|Float|Truncate (ScalableFloat|GetValueatLevel (Variables|Default|GetSelfPropagationFuelRowValue)))))
