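; FringePlank pulse indicator actor for a prototype weapon, which triggers the visual cue for a pulse-marking effect.
; When pulse markers start, it runs the parent logic, then the authority executes a replicated GameplayCue at the actor's location, with the instigator as source object.
; The marking radius from a scalable float row travels as the cue's raw magnitude.
; A comment marks this cue call as a fix for grid FX not playing.
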(event EventBeginPlay
  (|CallParentFunction))

(event EventStartPulseMarkers
  (|CallParentFunction)
  (Networking|SwitchHasAuthority
    (:Authority
      ;[1 1 1] Fix for FORT-558364 - grid FX not playing
      (Ability|ExecuteGameplayCueWithParamsReplicated (Variables|Default|GetGameplay%20Cue%20Tag) (Ability|GameplayCue|MakeGameplayCueParameters :NormalizedMagnitude 0.0 :RawMagnitude (ScalableFloat|GetValueatLevel (Variables|Default|GetRow_MarkingRadius)) :Location (Transformation|GetActorLocation) :Normal "0,0,0" :SourceObject (Game|GetInstigator) :GameplayEffectLevel 1 :AbilityLevel 1 :bReplicateLocationWhenUsingMinimalRepProxy false)))))

(event Collision|EventActorBeginOverlap
  (params (OtherActor :type Actor%20Object%20Reference)))

(event EventTick
  (params (DeltaSeconds :type Float%20%28single-precision%29)))

(fn ConstructionScript
  (|CallParentFunction))
