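; Weapon blueprint for the prototype FringePlank in Athena, handling its rear muzzle effect and out-of-ammo disposal.
; While equipped, it binds OnConsumeAmmo and, on authority, sends an AmmoChanged gameplay event to the owning pawn.
; The event magnitude carries remaining ammo, with an infinite-use override, because client ammo lags the server when the ability checks it for montage and discard.
; Local unequip removes an empty weapon from inventory as a safety net.
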(event EventMuzzleFlashFX
  (params (Persistent%20Fire :type Boolean))
  ;[1 1 1] Event -> Muzzle Flash FX
  (|CallParentFunction Persistent%20Fire)
  (Components|Activation|ToggleActive (Variables|Default|GetRearMuzzle)))

(event FortWeapon|EventOnEquippedAndAttached
  ;[0.5947 1 0.825] On ammo changed event to send gameplay event to ability
  (|CallParentFunction)
  (EventDispatchers|BindEventtoOnConsumeAmmo (EventDispatchers|CreateEvent :!event Custom|AmmoCountChangedGarbage)))

(event FortWeapon|EventOnUnEquippedAndDetached
  ;[0.5947 1 0.825] On ammo changed event to send gameplay event to ability
  (|CallParentFunction)
  (EventDispatchers|UnbindEventfromOnConsumeAmmo (EventDispatchers|CreateEvent :!event Custom|AmmoCountChangedGarbage)))

(event Custom|AmmoCountChangedGarbage
  (params (Quantity :type Integer))
  ;[0.5947 1 0.825] On ammo changed event to send gameplay event to ability
  (Networking|SwitchHasAuthority
    (:Authority
      (bind _asfort_player_pawn (Utilities|Casting|CastToFortPlayerPawn (Game|GetInstigator)))
      ;[0.01042 1 1] We check ammo in the GA this event triggers, but the client ammo count is not synced with server when we check to play montage and discard the weapon.
      ; Passing through ammo as event magnitude lets us properly dispose of the weapon
      (Ability|SendGameplayEventtoActor _asfort_player_pawn "(TagName=\"Event.Athena.Item.FringePlank.AmmoChanged\")" (Utilities|Struct|MakeGameplayEventData "(TagName=\"\")" 0 0 0 0 "()" "(GameplayTags=)" "(GameplayTags=)" (Math|Conversions|ToFloat(Integer) (select (FortWeapon|Ammo|HasInfiniteUses) 1 (FortWeapon|GetRemainingAmmo))))))))

(event FortWeapon|EventOnLocalUnEquip
  ;[1 1 1] Unequipped - Safety (if unequipped with no ammo...somehow, remove item)
  (CallFunction|RemoveItemOnNoAmmoRemaining))

(event Custom|Remove%20Item%20on%20No%20Ammo%20Remaining
  ;[1 1 1] Persist in Inventory when final stack empty in WID does not handle removing a custom ammo weapon with 0 ammo
  (if (<= (FortWeapon|GetRemainingAmmo) 0)
    (bind _asfort_player_pawn (Utilities|Casting|CastToFortPlayerPawn (Game|GetInstigator)))
    (Inventory|RemoveItemFromPlayerByGuid (Fortnite|GetFortPlayerControllerfromActor _asfort_player_pawn) (FortWeapon|GetInventoryGUID) 1 true)))

(fn ConstructionScript
  (|CallParentFunction))
