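; Secondary projectile Blueprint for the prototype FringePlank ranged weapon, a fire-trailed firework ember that explodes on impact.
; A BeginPlay timer triggers Burnout, which switches off both trails and clears the fire-spawn flag.
; Stopping on anything but a water body explodes it with an impact sound; entering water plays a one-time splash and cuts the ember trail.
; Explosion cosmetics stop the trail, play surface-based sound, show a sound indicator and fade looping audio.
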(event EventBeginPlay
  (Utilities|Time|SetTimerbyEvent (delegate Custom|Burnout) (Variables|Default|GetBurnoutTime)))

(event Custom|Burnout
  (Components|Activation|Deactivate (Variables|Default|GetFireTrail))
  (Components|Activation|Deactivate (Variables|Default|GetNS_Weapon_Firework_Ember_Trail))
  (Variables|Default|SetShouldSpawnFire false))

(event EventOnStop
  (params (Hit :type Hit%20Result%20Structure%20%28by%20ref%29))
  (bind (_bblockinghit _binitialoverlap _time _distance _location _impactpoint _normal _impactnormal _physmat _hitactor) (Collision|BreakHitResult Hit))
  ;[1 1 1] EVENT -> OnStop
  (|CallParentFunction Hit)
  (Variables|Default|SetImpact%20Point _impactpoint)
  (Utilities|Casting|CastToWaterBody _hitactor
    (:CastFailed
      (Delivery|DoExplosion)
      (CallFunction|PlayImpactSound))))

(event Custom|Play%20Impact%20Sound
  ;[1 1 1] EVENT -> OnStop
  (Audio|PlaySoundatLocation (Variables|Sounds|GetImpact%20Sound) (Variables|Default|GetImpact%20Point)))

(event OnEnterWaterEvent(WaterInteractionComponent)
  (params (WaterBody :type Fort%20Water%20Body%20Actor%20Object%20Reference) (WaterInteractionComponent :type Fort%20Water%20Interaction%20Component%20Object%20Reference) (bIsFirstBody :type Boolean))
  ;[0.04231 0.3916 1] EVENT -> On Enter Water
  (Utilities|FlowControl|DoOnce false
    (:Completed
      (Effects|SpawnEmitteratLocation (Variables|Effects|GetIntoWaterEffects) (Transformation|GetActorLocation))
      (Audio|PlaySoundatLocation (Variables|Sounds|GetHitWater_NoFire_Sound) (Transformation|GetActorLocation))
      (Components|Activation|Deactivate (Variables|Default|GetNS_Weapon_Firework_Ember_Trail)))))

(event EventOnExploded
  (params (HitActors :type Array%20of%20Actor%20Object%20References) (HitResults :type Array%20of%20Hit%20Result%20Structures))
  (bind _returnvalue (Transformation|GetActorLocation))
  (bind _loopingaudiotell (Variables|Default|GetLoopingAudioTell))
  ;[1 0.5938 0.5938] EVENT -> OnExploded
  (|CallParentFunction HitActors HitResults)
  (Networking|CanExecuteCosmeticEvents
    (:True
      (Components|Activation|SetActive (Variables|Default|GetNS_Weapon_Firework_Ember_Trail) false false)
      (Components|Activation|Deactivate (Variables|Default|GetNS_Weapon_Firework_Ember_Trail))
      (Class|BPSurfaceTypeSoundComponent|PlaySound (Variables|Default|GetBP_SurfaceTypeSoundComponent) _returnvalue)
      (Fortnite|ShowSoundIndicatorLocal :Location _returnvalue :MaxAudibleDistance 5e+03 :IndicatorType "Gunshot" :Instigator self :Tint "(R=0.000000,G=0.000000,B=0.000000,A=0.000000)" :AttenuationMaxRangeMultiplier 1.0 :AttenuationIntensityMultiplier 1.0)
      (if (Audio|Components|Audio|IsPlaying _loopingaudiotell)
        (Audio|Components|Audio|FadeOut _loopingaudiotell 1.5 0.0 "Linear" :!class /Script/Engine.AudioComponent)))))

(fn ConstructionScript
  (Actor|Tick|SetActorTickEnabled false))
