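; FringePlank flare gun projectile that bursts into sparks mid-flight, then slows to drift down under parachute gravity.
; A timer triggers the one-shot burst, which also spawns a server-side pulse indicator marker, while a timeline eases velocity if unbounced.
; Bounces damp speed and can spawn fire actors; player hits burst and destroy it.
; Bounce limits, landing after slowdown and water entry can end it, using delayed kills for effects.
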(event EventBeginPlay
  ;[0.5925 1 0.6093] EVENT -> OnBeginPlay
  (bind _asathena_player_controller (Utilities|Casting|CastToAthena_PlayerController (Fortnite|GetFortPlayerControllerfromActor (Variables|Actor|GetInstigator)))
    (:then
      (Variables|Default|SetAthenaPlayerControler _asathena_player_controller))
    (:CastFailed
      _))
  (Utilities|Time|SetTimerbyEvent (delegate Custom|SparkAndSlowDown) (ScalableFloat|GetValueatLevel (Variables|DesignVariables|GetTimeBeforeSparkExplosion)))
  (CallFunction|SetInitialProjectileValues)
  (Variables|/Script/Engine.ProjectileMovementComponent|SetProjectileGravityScale (ScalableFloat|GetValueatLevel (Variables|DesignVariables|GetInitial_Gravity)) (Variables|Movement|GetProjectileMovementComponent))
  (Networking|SwitchHasAuthority
    (:Authority
      (bind _asfort_player_pawn (Utilities|Casting|CastToFortPlayerPawn (Game|GetInstigator)))
      (Ability|SendGameplayEventtoActor _asfort_player_pawn "(TagName=\"Event.Athena.Item.FringePlank.ProjSpawned\")") ; Notifies GA_FringePlank_PassiveFromEvent_Athena that a projectile has spawned.
    )))

(event Custom|SparkAndSlowDown
  ;[0.5925 1 0.6093] EVENT -> OnBeginPlay
  (CallFunction|SparkExplosion)
  (CallFunction|SlowSpeed))

(event Custom|StopLoopingAudioTell
  (Audio|Components|Audio|Stop (Variables|Default|GetLoopingAudioTell)))

(event EventDestroyed
  (CallFunction|StopLoopingAudioTell)
  (bind _parachute_audio (Variables|Default|GetParachute%20Audio)
    (:then
      (Audio|Components|Audio|FadeOut _parachute_audio 0.25 0.0 "Linear" :!class /Script/Engine.AudioComponent))
    (:else)))

(event Custom|SetInitialProjectileValues
  ;[1 1 1] Set initial Projectile Speed
  (Projectile|SetVelocity (Variables|Movement|GetProjectileMovementComponent) (* (Math|Vector|Normalize (Variables|/Script/Engine.MovementComponent|GetVelocity (Variables|Movement|GetProjectileMovementComponent))) (ScalableFloat|GetValueatLevel (Variables|DesignVariables|GetInitial_Speed)))))

(event EventOnStop
  (params (Hit :type Hit%20Result%20Structure%20%28by%20ref%29))
  (bind (_bblockinghit _binitialoverlap _time _distance _location _impactpoint _normal _impactnormal _physmat _hitactor) (Collision|BreakHitResult Hit))
  (Variables|Default|SetImpact%20Point _impactpoint)
  (Utilities|Casting|CastToWaterBody _hitactor
    (:CastFailed
      (CallFunction|PlayImpactSound))))

(event Custom|Play%20Impact%20Sound
  (CallFunction|StopLoopingAudioTell)
  (Audio|PlaySoundatLocation (Variables|Sounds|GetImpact%20Sound) (Variables|Default|GetImpact%20Point))
  (Utilities|FlowControl|Delay 1.0)
  (Components|Activation|Activate (Variables|Default|GetNS_Emote_SingAlong_Book) false)
  (bind (_direction _meshscale) (Timeline|DestroyedMesh PlayFromStart)
    (:Update
      (Transformation|SetWorldScale3D (Variables|Projectile_GrenadeLauncher|GetMesh) (* "2.000000,2.000000,2.000000" _meshscale)))))

(event Custom|SparkExplosion
  ;[1 1 1] Spark Explosion (When it shoots sparks, not when it dies)
  (Utilities|FlowControl|DoOnce false
    (:Completed
      (if (not (or (Delivery|IsBeingKilled) (Variables|StateTrackers|GetinWater)))
        (Delivery|DoExplosion)
        (Niagara|SpawnSystematLocation :SystemTemplate (Variables|Default|GetSparkExplosion_NS) :Location (Transformation|GetActorLocation) :Scale (Math|Conversions|ToVector(Float) (Variables|Effects|GetSparkExplosionSize)) :bAutoDestroy true :bAutoActivate true :PoolingMethod "AutoRelease" :bPreCullCheck false)
        (Audio|PlaySoundatLocation (Variables|Sounds|GetSparkExplosion_Sound) (Transformation|GetActorLocation))
        (Components|Activation|Activate (Variables|Default|GetGiantFlare) false)
        (CallFunction|SlowSpeed)
        (CallFunction|SendOutPing)))))

(event Custom|SlowSpeed
  ;[1 1 1] Slow Down, Fall, and Deploy Parachute
  (if (== (Variables|StateTrackers|GetCurrentBounces) 0)
    (if (not (Variables|StateTrackers|GetisSlowing))
      (Variables|StateTrackers|SetisSlowing true)
      (Utilities|FlowControl|Sequence
        (:then_0
          (Variables|CachedValues|SetLerp_StartVel (Variables|/Script/Engine.MovementComponent|GetVelocity (Variables|Movement|GetProjectileMovementComponent)))
          (bind _output_get (Variables|CachedValues|SetLerp_StartXYSpeed (Math|Vector|VectorLengthXY (Variables|CachedValues|GetLerp_StartVel))))
          (Variables|CachedValues|SetLerp_FinalXYSpeed (* _output_get (ScalableFloat|GetValueatLevel (Variables|DesignVariables|GetFinal_XYSpeedMultiplier))))
          (Variables|CachedValues|SetLerp_StartZSpeed (.z (Variables|CachedValues|GetLerp_StartVel)))
          (Variables|CachedValues|SetLerp_XYDirection (Math|Vector|Normalize2D(Vector) (Variables|CachedValues|GetLerp_StartVel))))
        (:then_1
          (bind (_direction _velocitycurve) (Timeline|VelocityAndGravityLerp Play :NewTime 1.5)
            (:Update
              (bind (_returnvalue_x _returnvalue_y) (Math|Vector|BreakVector (* (Variables|CachedValues|GetLerp_XYDirection) (Math|Float|Lerp (Variables|CachedValues|GetLerp_StartXYSpeed) (Variables|CachedValues|GetLerp_FinalXYSpeed) _velocitycurve))))
              (Projectile|SetVelocity (Variables|Movement|GetProjectileMovementComponent) (Math|Vector|MakeVector _returnvalue_x _returnvalue_y (Math|Float|Lerp (Variables|CachedValues|GetLerp_StartZSpeed) 0.0 _velocitycurve))))
            (:Finished
              (Variables|/Script/Engine.ProjectileMovementComponent|SetProjectileGravityScale (ScalableFloat|GetValueatLevel (Variables|DesignVariables|GetGravityAfterParachute)) (Variables|Movement|GetProjectileMovementComponent)))))
        (:then_2
          (if (and (== (Variables|StateTrackers|GetCurrentBounces) 0) (not (Variables|StateTrackers|GetinWater)))
            (Variables|/Script/Engine.ProjectileMovementComponent|SetbRotationFollowsVelocity true (Variables|Movement|GetProjectileMovementComponent))))))))

(event Custom|HaltVelocityCurve
  ;[1 1 1] Slow Down, Fall, and Deploy Parachute
  (Variables|/Script/Engine.ProjectileMovementComponent|SetProjectileGravityScale (ScalableFloat|GetValueatLevel (Variables|DesignVariables|GetGravityAfterBounced)) (Variables|Movement|GetProjectileMovementComponent))
  (Timeline|VelocityAndGravityLerp Stop))

(event EventOnBounce
  (params (Hit :type Hit%20Result%20Structure%20%28by%20ref%29))
  (bind (_bblockinghit _binitialoverlap _time _distance _location _impactpoint _normal _impactnormal _physmat _hitactor) (Collision|BreakHitResult Hit))
  (bind _returnvalue_location (Math|Transform|BreakTransform (Transformation|GetActorTransform)))
  ;[1 1 1] EVENT -> OnBounce
  (Variables|StateTrackers|SetMarkPlayers false)
  (if (not (Variables|StateTrackers|GetIsDying))
    (Utilities|FlowControl|Sequence
      (:then_0
        (Utilities|Casting|CastToFortPlayerPawn _hitactor
          (:then
            (CallFunction|SparkExplosion)
            (CallFunction|StopLoopingAudioTell)
            (Actor|DestroyActor))
          (:CastFailed
            (Variables|/Script/Engine.ProjectileMovementComponent|SetbRotationFollowsVelocity true (Variables|Movement|GetProjectileMovementComponent))
            (if (Variables|StateTrackers|GetisSlowing)
              (if (not (Delivery|IsBeingKilled))
                (Utilities|FlowControl|Sequence
                  (:then_0
                    (Projectile|SetVelocity (Variables|Movement|GetProjectileMovementComponent) (0.0 0.0 0.0))
                    (CallFunction|KillwithLifeSpan 2.0)
                    (CallFunction|StopLoopingAudioTell)
                    (Audio|SpawnSoundatLocation (Variables|Sounds|GetTailStopSound) (Transformation|GetActorLocation)))
                  (:then_1
                    _)))
              (else
                (Utilities|FlowControl|Sequence
                  (:then_0
                    (CallFunction|SpawnBounceFX Hit))
                  (:then_1
                    (CallFunction|HaltVelocityCurve)
                    (Projectile|SetVelocity (Variables|Movement|GetProjectileMovementComponent) (* (Variables|/Script/Engine.MovementComponent|GetVelocity (Variables|Movement|GetProjectileMovementComponent)) (Math|Float|Clamp(Float) (+ (Math|Vector|DotProduct (Transformation|GetActorForwardVector) _impactnormal) 1.0) 0.3)))))))
            (if (ScalableFloat|GetValueasBool (Variables|Default|GetFlareGunCurieEnable))
              (Building|SpawnBuildingGameplayActor (Variables|Default|GetFireMeshBGA) (Math|Transform|MakeTransform _returnvalue_location) (Variables|Actor|GetInstigator) false false)))))
      (:then_1
        (Networking|SwitchHasAuthority
          (:Authority
            (CallFunction|IncrementBounces)))))))

(event Custom|IncrementBounces
  ;[1 1 1] EVENT -> OnBounce
  (Variables|StateTrackers|SetMarkPlayers false)
  (Math|Integer|IncrementInt (Variables|StateTrackers|GetCurrentBounces)
    (:%20%20
      (if (> (Variables|StateTrackers|GetCurrentBounces) (Math|Float|Truncate (ScalableFloat|GetValueatLevel (Variables|DesignVariables|GetMaxBounces))))
        (CallFunction|SparkExplosion)
        (Variables|StateTrackers|SetIsDying true)
        (CallFunction|StopLoopingAudioTell)
        (CallFunction|KillwithLifeSpan 0.4) ; Disables the projectile, but doesn't destroy until after the lifespan
      ))))

(event Custom|HitWaterEffects
  ;[0.2663 0.5273 1] EVENT -> On Enter Water
  (Effects|SpawnEmitteratLocation (Variables|Effects|GetWaterImpact_PS) (Transformation|GetActorLocation))
  (Audio|PlaySoundatLocation (Variables|Sounds|GetHitWater_Sound) (Transformation|GetActorLocation)))

(event OnEnterWaterEvent(WaterInteractionComponent)
  (params (WaterBody :type Fort%20Water%20Body%20Actor%20Object%20Reference) (WaterInteractionComponent :type Fort%20Water%20Interaction%20Component%20Object%20Reference) (bIsFirstBody :type Boolean))
  ;[0.2663 0.5273 1] EVENT -> On Enter Water
  (Variables|StateTrackers|SetinWater true)
  (if (< (Actor|GetGameTimeSinceCreation) 0.1)
    (CallFunction|HitWaterEffects)
    (Utilities|FlowControl|Delay 0.3)
    (if (Water|IsinWater (Variables|Water|GetWaterInteractionComponent))
      (CallFunction|StopLoopingAudioTell)
      (Actor|DestroyActor)
      (else
        (Variables|StateTrackers|SetinWater false)))
    (elif (Math|Boolean|ORBoolean (>= (Variables|StateTrackers|GetCurrentBounces) (Math|Float|Truncate (ScalableFloat|GetValueatLevel (Variables|DesignVariables|GetMaxBounces)))) (Variables|StateTrackers|GetisSlowing) (< (Utilities|VelocityDotwithUpVector) -0.8))
      (Utilities|FlowControl|DoOnce false
        (:Completed
          (Rendering|SetVisibility (Variables|Projectile_GrenadeLauncher|GetMesh) false false)
          (CallFunction|HitWaterEffects)
          (CallFunction|StopLoopingAudioTell)
          (CallFunction|KillwithLifeSpan 0.5) ; Disables the projectile, but doesn't destroy until after the lifespan.\nWait for particle effects to fizzle out
        ))
      (else
        (Networking|SwitchHasAuthority
          (:Authority)
          (:Remote
            (CallFunction|HitWaterEffects)))))))

(event Custom|SendOutPing
  ;[1 0.8512 0.4328] Send Out Ping
  (Networking|SwitchHasAuthority
    (:Authority
      (Utilities|FlowControl|DoOnce false
        (:Completed
          (bind _returnvalue (Building|SpawnBuildingGameplayActor (Variables|Default|GetBGAtoSpawn) (Transformation|GetActorTransform) (Variables|Actor|GetInstigator)))
          (bind _asbga_pulse_indicator (Utilities|Casting|CastToBGA_PulseIndicator _returnvalue))
          (bind _returnvalue_2 (CallFunction|GetOwnerWeapon))
          (Variables|/Game/Athena/Items/Consumables/PulseIndicator/BGA_PulseIndicator.BGA_PulseIndicator_C|SetOwnerWeapon _returnvalue_2 _asbga_pulse_indicator)
          (Class|BGAPulseIndicator|StartPulseMarkers _asbga_pulse_indicator)
          (Class|BGAPulseIndicator|Ready _asbga_pulse_indicator))))))

(fn KillWithLifeSpan
  (params (LifeSpan :type Float%20%28double-precision%29))
  (Variables|Delivery|SetLifespanAfterKill (Math|Float|Max(Float) LifeSpan 0.1))
  (Delivery|Kill))

(fn SpawnBounceFX
  (params (Hit :type Hit%20Result%20Structure%20%28by%20ref%29))
  (bind (_bblockinghit _binitialoverlap _time _distance _location _impactpoint _normal _impactnormal) (Collision|BreakHitResult (LocalVariables|GetBounceHit)))
  (bind _bouncehit (LocalVariables|GetBounceHit))
  (bind (_bblockinghit_2 _binitialoverlap_2 _time_2 _distance_2 _location_2 _impactpoint_2 _normal_2 _impactnormal_2 _physmat _hitactor) (Collision|BreakHitResult _bouncehit))
  (LocalVariables|SetBounceHit Hit)
  (Class|BPSurfaceTypeSoundComponent|SetSurfaceType (Variables|Default|GetBP_SurfaceTypeSoundComponent) (Collision|GetFortPhysicalSurface _bouncehit) _hitactor)
  (Fortnite|ShowSoundIndicatorLocal :Location _location_2 :MaxAudibleDistance 3.5e+03 :IndicatorType "Gunshot" :Instigator self :Tint "(R=0.000000,G=0.000000,B=0.000000,A=0.000000)" :AttenuationMaxRangeMultiplier 1.0 :AttenuationIntensityMultiplier 1.0)
  (bind _outresult (Audio|OnCollision (Variables|Default|GetFortCollisionAudio) _impactpoint (* _impactnormal (Variables|/Script/Engine.MovementComponent|GetVelocity (Variables|Movement|GetProjectileMovementComponent)))))
  (if _outresult
    (GameplayCue|ExecuteGameplayCue%28Burst%29Local self (Variables|Default|GetBounceCueTag) (Ability|GameplayCue|MakeGameplayCueParameters :NormalizedMagnitude 0.0 :RawMagnitude 0.0 :Location "0,0,0" :Normal "0,0,0" :Instigator (Game|GetInstigator) :GameplayEffectLevel 1 :AbilityLevel 1 :bReplicateLocationWhenUsingMinimalRepProxy false))
    (return)
    (else
      (return))))

(fn ConstructionScript
  (bind _initial_speed (Variables|DesignVariables|GetInitial_Speed))
  (bind _returnvalue (ScalableFloat|GetValueatLevel _initial_speed))
  (Effects|Components|ParticleSystem|SetTemplate (Variables|Default|GetRibbonTrail_PSC) (Variables|Effects|GetRibbonTrail_PS))
  (Audio|SetCollisionSoundatIndex (Variables|Default|GetFortCollisionAudio) 0 (Variables|Sounds|GetBounce_Sound))
  (Variables|/Script/Engine.ProjectileMovementComponent|SetInitialSpeed _returnvalue (Variables|Movement|GetProjectileMovementComponent))
  (Variables|/Script/Engine.ProjectileMovementComponent|SetMaxSpeed _returnvalue (Variables|Movement|GetProjectileMovementComponent))
  (Actor|Tick|SetActorTickEnabled false))
